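# Animation Blueprint for the male medium Backspin body cosmetic, copying the attached parent's pose and adjusting knife and pouch bones on top.
# A Bone Driven Controller turns dyn_knife on Y by a fraction of the right thigh's Y rotation.
# An additive bone-space Modify Bone then offsets and rotates dyn_parent_pouch, weighted by crouch_alpha.
# The crouch begin and end events switch crouch_alpha between one and zero.
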AnimBlueprint /BRCosmetics/Characters/Player/Male/Medium/Bodies/M_MED_Backspin/Meshes/M_MED_Backspin_AnimBP
target_skeleton=&/M_MED_Backspin_Skeleton
graphs [
 AnimGraph {nodes [
   n1 = "dyn_knife.rotateY = thigh_r.rotateY * 0.4 - Bone Driven Controller"!(
    ComponentPose:Component Space Pose Link Structure=Local To Component(LocalPose=Copy Pose From Mesh() {bUseAttachedParent=true})
    Alpha:float=1
   ) -> (Pose:Component Space Pose Link Structure) {
    is=title AnimGraphNode_BoneDrivenController
    bAffectTargetRotationY=true
    multiplier=0.4
    sourceComponent=RotationY
    alphaBoolBlend {blendOption=Linear}
    sourceBone {boneName=thigh_r}
    targetBone {boneName=dyn_knife}
   }
   n2 = "Transform (Modify) Bone - Bone: dyn_parent_pouch"(
    ComponentPose=@n1
    Alpha=Get crouch_alpha() -> (crouch_alpha:double)
    Translation="0.004,-0.853,5.686"
    Rotation="-0.915,-1.500,-31.345"
    Scale="1,1,1"
   ) {
    rotationMode=BMM_Additive
    rotationSpace=BCS_BoneSpace
    translationMode=BMM_Additive
    translationSpace=BCS_BoneSpace
    alphaBoolBlend {blendOption=Linear}
    alphaScaleBiasClamp {interpSpeedDecreasing=15, interpSpeedIncreasing=15}
    boneToModify {boneName=dyn_parent_pouch}
    rotation=<rot -0.915 -1.5 -31.3451>
    translation=<0.004 -0.853 5.686>
   }
   Output Pose(Result=Component To Local(ComponentPose=@n2)) {layerGroup=None}
  ]}
 EventGraph {nodes [
   Event OnBeginCrouching()
   > Set crouch_alpha(crouch_alpha:double=1) -> (Output_Get:double)
   Event OnEndCrouching()
   > Set crouch_alpha(crouch_alpha=0)
   Event BlueprintUpdateAnimation()
  ]}
]
variables [crouch_alpha {type=number}]
