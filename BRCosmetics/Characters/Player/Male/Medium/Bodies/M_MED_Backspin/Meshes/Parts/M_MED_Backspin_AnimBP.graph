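# Animation Blueprint for the M_MED_Backspin player cosmetic mesh part, following the parent body and adjusting its glasses.
# Copy Pose From Mesh feeds a glasses Control Rig, whose alpha is on only at LOD0, into a cached pose.
# While parts notifies with context Backspin_Idle set FrontEnd_Idle_On, a bool blend swaps in that pose with a small additive bone-space offset on dyn_glasses.
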AnimBlueprint /BRCosmetics/Characters/Player/Male/Medium/Bodies/M_MED_Backspin/Meshes/Parts/M_MED_Backspin_AnimBP
target_skeleton=&/M_MED_Backspin_Skeleton
graphs [
 AnimGraph {nodes [
   n1 = "Transform (Modify) Bone - Bone: dyn_glasses"!(ComponentPose=Local To Component(LocalPose=Use cached pose 'CachePose'()), Alpha=1, Translation="0,0.45,0", Rotation="0,4,0") -> (Pose) {
    rotationMode=BMM_Additive
    rotationSpace=BCS_BoneSpace
    translationMode=BMM_Additive
    translationSpace=BCS_BoneSpace
    alphaBoolBlend {blendOption=Linear}
    boneToModify {boneName=dyn_glasses}
    rotation {yaw=4}
    translation {y=0.45}
   }
   Output Pose(Result=Blend Poses by bool(
     BlendPose_0=Component To Local(ComponentPose=@n1)
     BlendPose_1=Use cached pose 'CachePose'()
     bActiveValue=Get FrontEnd_Idle_On() -> (FrontEnd_Idle_On:bool)
    ) {blendType=Linear}) {layerGroup=None}
   n2 = Control Rig(Source=Copy Pose From Mesh() {bCopyCurves=true, bUseAttachedParent=true}, Alpha=Get isLOD0() -> (isLOD0:double)) {
    alphaBoolBlend {blendOption=Linear}
    controlRigAssetReference {blueprintRigClass=&/M_MED_Backspin_Glasses_CtrlRig.M_MED_Backspin_Glasses_CtrlRig_C}
   }
   Save cached pose 'CachePose'(Pose=@n2) {cacheName=CachePose}
  ]}
 EventGraph {nodes [
   Event BlueprintUpdateAnimation()
   > n1 = Branch(Condition="Equal (Integer)"(A=Get CurrentLOD(), B=0) {is=title K2Node_CallFunction})
   > Set isLOD0(isLOD0:double=1) -> (Output_Get:double)
   Set isLOD0(execute=@n1.else, isLOD0=0)
   n2 = Event OnNotifyPartsStarted!() -> (OutputDelegate, then, Context:Name=None) {is=title K2Node_Event}
   > Branch(Condition="Equal (Name)"!(A:Name=@n2.Context, B:Name=Backspin_Idle) -> (ReturnValue:bool) {is=title K2Node_CallFunction})
   > Set FrontEnd_Idle_On(FrontEnd_Idle_On:bool=true) -> (Output_Get:bool)
   n3 = Event OnNotifyPartsStopped!() -> (OutputDelegate, then, Context:Name=None) {is=title K2Node_Event}
   > Branch(Condition="Equal (Name)"(A=@n3.Context, B=Backspin_Idle))
   > Set FrontEnd_Idle_On(FrontEnd_Idle_On=false)
   TryGetPawnOwner()
  ]}
]
variables [
 FrontEnd_Idle_On {type=boolean}
 isLOD0 {type=number}
]
